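# Material function for the SoapPocket emote prop that toasts and singes its surface, overriding Base Color in the incoming material attributes.
# Distance-field-style (DF) thresholds driven by Toast Time and Singe Time are broken up with macro noise and contrast to form burn masks.
# Base color height-lerps from the untoasted texture toward a toast color, then blends toward the singed look.
# A sine-shaped orange burn glow is also computed, but the output sets no emissive.
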MaterialFunction /BRCosmetics/Animation/Props/Gadgets_SoapPocket/Materials/MF_Emote_Prop_SoapPocket
graph {nodes [
  n1 = GetMaterialAttributes(None=FunctionInput() {inputName=Result, inputType=FunctionInput_MaterialAttributes}) {attributeGetTypes [#1]}
  FunctionOutput(None=SetMaterialAttributes!(MaterialAttributes=@n1.MaterialAttributes, BaseColor=NamedRerouteUsage()) {attributeSetTypes [#1]}) {outputName=SetMaterialAttributes}
  NamedRerouteDeclaration(None=@n1.BaseColor) {name=Base Color - In, nodeColor=<rgba 0 1 0.7188 1>}
  n2 = Multiply(A=TextureCoordinate(), B=ScalarParameter() {defaultValue=0.2, parameterName=Noise Tiling})
  n3 = TextureSample(UVs=@n2) {texture=/Engine/EngineMaterials/T_Default_MacroVariation}
  n4 = TextureSample() {samplerType=SAMPLERTYPE_Grayscale, texture=/BRCosmetics/Animation/Props/Gadgets_SoapPocket/Textures/T_Emote_SoapPocket_DF}
  n5 = SmoothStep(
   Min=@n4.R
   Max=Add(A=@n4.R, B=ScalarParameter() {defaultValue=0.7, parameterName=Singe - Transition Fade}) {constB=0.5}
   Value=ScalarParameter() {defaultValue=0.3394, parameterName=Singe Time, sliderMax=1.05}
  ) {constValue=0.1}
  n6 = ScalarParameter() {parameterName="Contrast (S)"}
  n7 = MaterialFunctionCall(In=Multiply(A=Add(A=@n3.RGB, B=@n5), B=@n5), Contrast=@n6) {materialFunction=&CheapContrast}
  NamedRerouteDeclaration(None=Saturate(None=Power(Base=@n7.Result) {constExponent=1})) {name=Burn Mask, nodeColor=<rgba 1 0.3984 0 1>}
  n8 = TextureSample() {texture=/BRCosmetics/Animation/Props/Gadgets_SoapPocket/Textures/T_SoapPocket_Untoasted_Emote_Prop_D}
  n9 = VectorParameter() {parameterName=Toast Color, defaultValue=<rgba 0.5625 0.2036 0 1>}
  n10 = OneMinus(None=NamedRerouteUsage())
  n11 = Add(A=@n10, B=@n3.RGB)
  n12 = Clamp(None=Multiply(A=Multiply(A=@n11, B=@n10), B=ScalarParameter() {defaultValue=0.432, parameterName=Toast Time})) {maxDefault=0.7}
  n13 = NamedRerouteUsage()
  n14 = ScalarParameter() {defaultValue=0.5, parameterName=Toasted - Height Lerp - Contrast}
  n15 = NamedRerouteUsage()
  n16 = ScalarParameter() {defaultValue=0.5, parameterName=Singe Exponent}
  NamedRerouteDeclaration(None=LinearInterpolate(
    A=MaterialFunctionCall(A=@n8.RGB, B=@n9.RGB, Transition Phase=@n12, Height Texture=@n13, Contrast=@n14) {materialFunction=&HeightLerp}.Results
    B=NamedRerouteUsage()
    Alpha=Power(Base=@n15, Exp=@n16)
   )) {name=Base Color - Out, nodeColor=<rgba 0.0781 0 1 1>}
  n17 = Constant()
  n18 = MaterialFunctionCall(
   A=@n17
   B=VectorParameter() {parameterName=Emissive 2, defaultValue=<rgba 50 8.4778 0 1>}.RGB
   C=@n17
   Alpha=Power(Base=@n7.Result) {constExponent=6}
  ) {materialFunction=&3ColorBlend}
  n19 = ScalarParameter() {defaultValue=500, parameterName=Burn Brightness}
  n20 = NamedRerouteUsage()
  NamedRerouteDeclaration(None=LinearInterpolate(B=Multiply(A=@n18.Result, B=@n19), Alpha=Sine(None=@n20)) {constB=0}) {name=Emissive - Out, nodeColor=<rgba 0.2422 1 0 1>}
  n21 = TextureSample() {texture=/BRCosmetics/Animation/Props/Gadgets_SoapPocket/Textures/T_SoapPocket_Emote_Prop_Mask}
  NamedRerouteDeclaration(None=@n21.B) {name=Dragon Mask, nodeColor=<rgba 1 0.6563 0 1>}
  NamedRerouteDeclaration(None=@n21.R) {name=DF, nodeColor=<rgba 0.3359 0 1 1>}
  NamedRerouteDeclaration(None=@n21.G) {name=Height, nodeColor=<rgba 1 0 0.0703 1>}
  n22 = SmoothStep(
   Min=@n4.R
   Max=Add(A=@n4.R, B=ScalarParameter() {defaultValue=1, parameterName=Toast - Transition Fade}) {constB=0.5}
   Value=ScalarParameter() {defaultValue=0.8, parameterName=Toast Time}
  ) {constValue=0.1}
  n23 = MaterialFunctionCall(In=Multiply(A=Add(A=@n3.RGB, B=@n22), B=@n22), Contrast=@n6) {materialFunction=&CheapContrast}
  NamedRerouteDeclaration(None=Saturate(None=Power(Base=@n23.Result) {constExponent=0.5})) {name=Toast Burn Mask, nodeColor=<rgba 1 0.3984 0 1>}
  NamedRerouteUsage()
  Panner(Coordinate=@n2) {speedX=0.01, speedY=0.01}
  Power(Base=@n3.RGB) {constExponent=0.5}
 ]}
